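; Weapon Blueprint for the Cerberus break-action shotgun that supplies its shell visual effects.
; ActivateShellsFX, when passed true, spawns the Cerberus shell burst Niagara system attached to the weapon mesh at the shells socket and stores it in Spawned_Shells.
; It then passes bIsLocal to the system's IsSystemLocal parameter, which is used for shell audio.
; The construction script only calls its parent.
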(fn ActivateShellsFX
  (params (Bool :type Boolean))
  (if Bool
    (bind _returnvalue (Niagara|SpawnSystemAttached "/SunRoseWeaponsGameplay/Items/Weapons/CerberusSG/FX/NS_CerberusSG_ShellsFire_Burst.NS_CerberusSG_ShellsFire_Burst" (Variables|Visual|GetWeaponMesh) (Variables|Effects|Shells|GetShells%20Socket%20Name) "0, 0, 0" "0, 0, 0" "KeepRelativeOffset" true))
    (bind _output_get (Variables|Effects|Shells|SetSpawned_Shells _returnvalue))
    ;[0 0 0] Used for shell audio
    (Niagara|SetNiagaraVariable(Bool) _output_get "IsSystemLocal" (Variables|Default|GetbIsLocal))))

(fn ConstructionScript
  (|CallParentFunction))
