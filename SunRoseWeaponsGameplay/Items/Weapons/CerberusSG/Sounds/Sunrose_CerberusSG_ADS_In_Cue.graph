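# Sound cue for the Cerberus shotgun's aim-down-sights-in foley in the SunRose weapons set.
# A Random node picks among three ADS-in wave variations with equal weight, feeding a Modulator that raises the upper pitch limit for variation.
# It plays through the weapon foley sound class with weapon reload attenuation at reduced volume.
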SoundCue /SunRoseWeaponsGameplay/Items/Weapons/CerberusSG/Sounds/Sunrose_CerberusSG_ADS_In_Cue
root=Modulator_0
nodes [
 {id=Modulator_0, children [Random_1], properties {pitchMax=1.2}}
 {id=Random_1, children [WavePlayer_0,WavePlayer_1,WavePlayer_3], properties {weights [1,1,1]}}
 {id=WavePlayer_0, properties {soundWaveAssetPtr=&/Sunrose_CerberusSG_ADS_In_01}}
 {id=WavePlayer_1, properties {soundWaveAssetPtr=&/Sunrose_CerberusSG_ADS_In_02}}
 {id=WavePlayer_3, properties {soundWaveAssetPtr=&/Sunrose_CerberusSG_ADS_In_03}}
]
properties {
 attenuationSettings=&Fort_Weapon_Reload_Att
 maxDistance=2250
 soundClassObject=&Fort_WeaponFoley
 volumeMultiplier=0.501
}
