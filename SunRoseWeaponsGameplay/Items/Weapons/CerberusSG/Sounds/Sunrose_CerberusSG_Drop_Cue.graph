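# Sound cue for dropping the Sunrose CerberusSG weapon.
# A single drop wave plays through a modulator that slightly varies pitch and volume, and the cue applies a volume boost.
# It uses the shared weapon drop attenuation, the loot collect sound class and a stop-oldest weapon drop concurrency set.
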SoundCue /SunRoseWeaponsGameplay/Items/Weapons/CerberusSG/Sounds/Sunrose_CerberusSG_Drop_Cue
root=Modulator_0
nodes [
 {id=Modulator_0, children [WavePlayer_1], properties {pitchMax=1, pitchMin=0.9, volumeMax=0.9, volumeMin=0.8}}
 {id=WavePlayer_1, properties {soundWaveAssetPtr=&/Sunrose_CerberusSG_Drop_01}}
]
properties {
 attenuationSettings=&Fort_Weapon_Drop_Att
 maxDistance=3000
 soundClassObject=&Fort_Loot_Collect
 volumeMultiplier=2
 concurrencyOverrides {resolutionRule=PreventNew}
 concurrencySet [&Weapon_Drop_StopOldest]
}
