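# Sound cue for the magazine-insert reload foley of the SunRose Cerberus SG weapon.
# A single wave player plays one mag-insert recording through a modulator that slightly randomizes pitch and volume on each play.
# It uses the shared weapon reload attenuation and weapon foley sound class, and overrides concurrency.
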SoundCue /SunRoseWeaponsGameplay/Items/Weapons/CerberusSG/Sounds/Sunrose_CerberusSG_Mag_Insert_Cue
root=Modulator_0
nodes [
 {id=Modulator_0, children [WavePlayer_5], properties {pitchMax=1.1, pitchMin=0.9, volumeMax=1.1, volumeMin=0.9}}
 {
  id=WavePlayer_5
  properties {soundWaveAssetPtr=&/Sunrose_CerberusSG_Reload_MagInsert_01}
 }
]
properties {
 attenuationSettings=&Fort_Weapon_Reload_Att
 bOverrideConcurrency=true
 maxDistance=2250
 soundClassObject=&Fort_WeaponFoley
 volumeMultiplier=1.002
}
