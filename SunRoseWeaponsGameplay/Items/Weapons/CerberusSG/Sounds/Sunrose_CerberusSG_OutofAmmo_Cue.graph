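# Sound cue for the out-of-ammo dry-fire sound of the Cerberus SG weapon in the SunRose gameplay content.
# A single wave player feeds a modulator that slightly randomizes volume at fixed pitch, routed to a third-person ranged weapon fire sound class.
# Concurrency overrides allow one instance per owner, preventing new plays while one is active and setting a short retrigger time to limit spam.
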SoundCue /SunRoseWeaponsGameplay/Items/Weapons/CerberusSG/Sounds/Sunrose_CerberusSG_OutofAmmo_Cue
root=Modulator_0
nodes [
 {id=Modulator_0, children [WavePlayer_0], properties {pitchMax=1, pitchMin=1, volumeMax=1.1, volumeMin=1}}
 {id=WavePlayer_0, properties {soundWaveAssetPtr=&/Sunrose_CerberusSG_Dryfire_01}}
]
properties {
 bOverrideConcurrency=true
 maxDistance=2097152
 soundClassObject=&Fort_WeaponRanged_Fire_3P_PSM
 volumeMultiplier=1.5
 attenuationOverrides {falloffDistance=1600}
 concurrencyOverrides {bLimitToOwner=true, maxCount=1, resolutionRule=PreventNew, retriggerTime=0.3}
}
