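# Sound cue that plays the pickup sound for the Sunrose CerberusSG weapon.
# A single wave player feeds a modulator that slightly randomizes volume on each play.
# It uses the shared weapon pickup attenuation and the loot collect sound class, with a boosted volume multiplier.
# It references the oldest-stopping pickup concurrency set and carries single-instance, prevent-new override values.
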SoundCue /SunRoseWeaponsGameplay/Items/Weapons/CerberusSG/Sounds/Sunrose_CerberusSG_Pickup_Cue
root=Modulator_1
nodes [
 {id=Modulator_1, children [WavePlayer_0], properties {volumeMax=0.9, volumeMin=0.8}}
 {id=WavePlayer_0, properties {soundWaveAssetPtr=&/Sunrose_CerberusSG_Pickup_01}}
]
properties {
 attenuationSettings=&Fort_Weapon_Pickup_Att
 maxDistance=3500
 soundClassObject=&Fort_Loot_Collect
 volumeMultiplier=1.5
 concurrencyOverrides {maxCount=1, resolutionRule=PreventNew}
 concurrencySet [&PickupMaximum_Oldest]
}
